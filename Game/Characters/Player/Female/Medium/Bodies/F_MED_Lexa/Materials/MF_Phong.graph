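# MF_Phong is a material function in the Lexa character's Materials folder that computes a sun-driven specular highlight term.
# It adds the camera vector to the sunlight direction from FortniteMaterialParameters, normalizes it, and dots it with the pixel normal, a Blinn-Phong half-vector technique.
# The output is that raw, unclamped dot product.
# A SpecPower exponent and Sharpness-driven CheapContrast branch exists but does not feed the output.
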MaterialFunction /Game/Characters/Player/Female/Medium/Bodies/F_MED_Lexa/Materials/MF_Phong
graph {nodes [
  n1 = CameraVectorWS()
  n2 = CollectionParameter() {
   collection=&FortniteMaterialParameters
   parameterName=SunlightDirectionalVector
  }
  n3 = ComponentMask(None=@n2) {b=true, g=true, r=true}
  n4 = DotProduct(A=Normalize(VectorInput=Add(A=@n1, B=@n3)), B=PixelNormalWS())
  FunctionOutput(None=@n4)
  MaterialFunctionCall(
   In=Power(Base=@n4, Exp=ScalarParameter() {defaultValue=16, parameterName=SpecPower})
   Contrast=ScalarParameter() {defaultValue=4, parameterName=Sharpness}
  ) {materialFunction=&CheapContrast}
 ]}
